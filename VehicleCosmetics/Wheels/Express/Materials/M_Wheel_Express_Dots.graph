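# Translucent unlit material for the animated dot-matrix effect on the Express cosmetic vehicle wheel.
# A dot-matrix VFX texture's channels mask and blend accent and custom colors into boosted emissive.
# Two opacity variants add time-scrolled sine pulses and sharp power falloffs, over a radial texture and linear UV.
# Vertex color red drives the emissive and opacity lerps, whose result passes through vehicle damage and de-rez transition functions.
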Material /VehicleCosmetics/Wheels/Express/Materials/M_Wheel_Express_Dots
graph {
 properties {
  bAutomaticallySetUsageInEditor=false
  bUseMaterialAttributes=true
  bUsedWithMeshDeformer=true
  bUsedWithSkeletalMesh=true
  blendMode=BLEND_Translucent
  previewMesh=/VehicleCosmetics/Wheels/Glitch/Meshes/SK_Wheel_Glitch
  shadingModel=MSM_Unlit
  usedShadingModels=Unlit
 }
 nodes [
  n1 = VertexColor()
  n2 = LinearInterpolate(B=NamedRerouteUsage(), Alpha=@n1.R)
  n3 = LinearInterpolate(A=NamedRerouteUsage(), B=NamedRerouteUsage(), Alpha=@n1.R)
  MaterialFunctionCall(In=MaterialFunctionCall(Material=SetMaterialAttributes!(MaterialAttributes, Emissive Color=@n2, Opacity=@n3) {attributeSetTypes [#1, #2]}) {materialFunction=&MF_BaseVehicle_Damage}.Result) {materialFunction=&MF_VehicleDeRez_Transition}
  n4 = NamedRerouteUsage()
  n5 = TextureSample() {texture=/VehicleCosmetics/Wheels/_SharedTextures/FX/Radial_Generic_01_Pack}
  n6 = Time()
  n7 = Add(A=@n5.G, B=@n6)
  n8 = Multiply(A=@n4, B=Multiply(A=@n5.R, B=Sine(None=@n7) {desc=Period 0.8, period=0.8}))
  n9 = Multiply(A=Multiply(A=NamedRerouteUsage(), B=Power(Base=@n5.G) {constExponent=24})) {constB=12}
  NamedRerouteDeclaration(None=Add(A=Add(A=@n8, B=@n9), B=Multiply(A=@n4, B=@n5.G))) {name=Opacity, nodeColor=<rgba 1 0 0.6797 1>}
  n10 = NamedRerouteUsage()
  n11 = ComponentMask(None=TextureCoordinate()) {r=true}
  n12 = Add(A=@n11, B=@n6)
  n13 = Multiply(A=@n10, B=Multiply(A=@n11, B=Sine(None=@n12) {desc=Period 1.5, period=1.5}))
  n14 = Multiply(A=Multiply(A=NamedRerouteUsage(), B=Power(Base=@n11) {constExponent=5})) {constB=26}
  NamedRerouteDeclaration(None=Add(A=Add(A=@n13, B=@n14), B=Multiply(A=@n10, B=@n13))) {name=Opacity_Linear, nodeColor=<rgba 1 0 0.6797 1>}
  n15 = TextureSampleParameter2D() {parameterName=VFX, texture=/VehicleCosmetics/Wheels/Express/Textures/Wheel_RazzleDazzle_DotMatrix_RGB}
  NamedRerouteDeclaration(None=@n15.A) {name=Mask_A, nodeColor=<rgba 0 0.6016 1 1>}
  n16 = Multiply(A=@n15.G, B=VectorParameter() {parameterName=Accent Color, defaultValue=<rgba 0.9216 0.2961 0.956 1>}.RGB)
  n17 = Multiply(A=@n15.B, B=VectorParameter() {parameterName=Custom Color, defaultValue=<rgba 0.0437 1 1 1>}.RGB)
  n18 = NamedRerouteDeclaration(None=@n15.R) {name=Mask_R, nodeColor=<rgba 1 0 0.2812 1>}
  NamedRerouteDeclaration(None=Multiply(A=LinearInterpolate(A=@n16, B=@n17, Alpha=@n18)) {constB=3}) {name=Color, nodeColor=<rgba 1 0 0.8671 1>}
 ]
}
